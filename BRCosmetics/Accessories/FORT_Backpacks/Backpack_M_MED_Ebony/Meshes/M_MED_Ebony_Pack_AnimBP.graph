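# Animation Blueprint for the Ebony backpack, a kill-reactive cosmetic that steps through deployment stages.
# An Idle/on_Kill state machine picks stage idle or transition clips by KillCounter, copied from the backpack's part modifier after a short retriggerable delay.
# Cosmetic variants enable reactivity or pin the final-stage idle, while material parameters reveal hidden elements at early stages.
# A looping frontend timer advances stages; computed turn/look correction alphas are unused in this AnimGraph.
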AnimBlueprint /BRCosmetics/Accessories/FORT_Backpacks/Backpack_M_MED_Ebony/Meshes/M_MED_Ebony_Pack_AnimBP
target_skeleton=/Game/Accessories/FORT_Backpacks/Backpack_M_MED_Ebony/Meshes/M_MED_Ebony_Pack_Skeleton
graphs [
 AnimGraph {nodes [
   AnimGraphNode_StateMachine_0 = New State Machine() {is=title AnimGraphNode_StateMachine}
   n1 = "Transform (Modify) Bone - Bone: root"(ComponentPose=Local To Component(LocalPose=@AnimGraphNode_StateMachine_0), Translation="0,-2,0", Rotation="0,0,0") {
    translationMode=BMM_Additive
    translationSpace=BCS_BoneSpace
    alphaBoolBlend {blendOption=Linear}
    boneToModify {boneName=root}
    translation {y=-2}
   }
   Output Pose(Result=Component To Local(ComponentPose=@n1)) {layerGroup=None}
  ]}
 New State Machine {
  entry=Idle
  path=AnimGraph.AnimGraphNode_StateMachine_0.New State Machine
  ids=id
  nodes [
   AnimStateNode_0 = Idle()
   AnimStateNode_1 = on_Kill() {is=title AnimStateNode}
   AnimStateTransitionNode_0 = AnimStateTransitionNode() {crossfadeDuration=0}
   AnimStateTransitionNode_1 = AnimStateTransitionNode() {bAutomaticRuleBasedOnSequencePlayerInState=true}
  ]
  links [
   {
    dst=on_Kill
    rule=AnimGraph.AnimGraphNode_StateMachine_0.New State Machine.AnimStateTransitionNode_0.Transition
    src=Idle
    via=AnimStateTransitionNode_0
   }
   {
    dst=Idle
    rule=AnimGraph.AnimGraphNode_StateMachine_0.New State Machine.AnimStateTransitionNode_1.Transition
    src=on_Kill
    via=AnimStateTransitionNode_1
   }
  ]
 }
 Idle {
  path=AnimGraph.AnimGraphNode_StateMachine_0.New State Machine.AnimStateNode_0.Idle
  nodes [
   n1 = Blend Poses by int!(
    BlendPose_0:Pose Link Structure=Sequence Player 'Ebony_Stage00_Idle_PROP'() {bLoopAnimation=false, sequence=/BRCosmetics/Accessories/FORT_Backpacks/Backpack_M_MED_Ebony/Animation/Ebony_Stage00_Idle_PROP}
    BlendPose_1:Pose Link Structure=Sequence Player 'Ebony_Stage01_Idle_PROP'() {bLoopAnimation=false, sequence=/BRCosmetics/Accessories/FORT_Backpacks/Backpack_M_MED_Ebony/Animation/Ebony_Stage01_Idle_PROP}
    BlendPose_2:Pose Link Structure=Sequence Player 'Ebony_Stage02_Idle_PROP'() {bLoopAnimation=false, sequence=/BRCosmetics/Accessories/FORT_Backpacks/Backpack_M_MED_Ebony/Animation/Ebony_Stage02_Idle_PROP}
    BlendPose_3:Pose Link Structure=Sequence Player 'Ebony_Stage03_Idle_PROP'() {bLoopAnimation=false, sequence=/BRCosmetics/Accessories/FORT_Backpacks/Backpack_M_MED_Ebony/Animation/Ebony_Stage03_Idle_PROP}
    BlendPose_4:Pose Link Structure=Sequence Player 'Ebony_Stage04_Idle_PROP'() {bLoopAnimation=false, sequence=/BRCosmetics/Accessories/FORT_Backpacks/Backpack_M_MED_Ebony/Animation/Ebony_Stage04_Idle_PROP}
    BlendPose_5:Pose Link Structure=Sequence Player 'Ebony_Stage05_Idle_PROP'() {bLoopAnimation=false, sequence=/BRCosmetics/Accessories/FORT_Backpacks/Backpack_M_MED_Ebony/Animation/Ebony_Stage05_Idle_PROP}
    BlendPose_6:Pose Link Structure=Sequence Player 'Ebony_Stage06_Idle_PROP'() {bLoopAnimation=false, sequence=/BRCosmetics/Accessories/FORT_Backpacks/Backpack_M_MED_Ebony/Animation/Ebony_Stage06_Idle_PROP}
    BlendTime_0:float=0.1
    BlendTime_1:float=0.1
    BlendTime_2:float=0.1
    BlendTime_3:float=0.1
    BlendTime_4:float=0.1
    BlendTime_5:float=0.1
    BlendTime_6:float=0.1
    ActiveChildIndex:int=Get KillCounter() -> (KillCounter:int=0)
   ) -> (Pose:Pose Link Structure) {
    is=title AnimGraphNode_BlendListByInt
    blendType=Linear
    blendPose [
     {linkId=-1, sourceLinkId=-1}
     {linkId=-1, sourceLinkId=-1}
     {linkId=-1, sourceLinkId=-1}
     {linkId=-1, sourceLinkId=-1}
     {linkId=-1, sourceLinkId=-1}
     {linkId=-1, sourceLinkId=-1}
     {linkId=-1, sourceLinkId=-1}
    ]
    blendTime [0.1,0.1,0.1,0.1,0.1,0.1,0.1]
   }
   Output Animation Pose(Result=Blend Poses by bool(
     BlendPose_0=Sequence Player 'Ebony_Stage06_Idle_PROP'() {bLoopAnimation=false, sequence=&Ebony_Stage06_Idle_PROP}
     BlendPose_1=@n1
     bActiveValue=Get IsNonReactiveFullyDeployed() -> (IsNonReactiveFullyDeployed:bool)
    ) {blendType=Linear})
  ]
 }
 on_Kill {
  path=AnimGraph.AnimGraphNode_StateMachine_0.New State Machine.AnimStateNode_1.on_Kill
  nodes [
   Sequence Player 'Ebony_toStage06_PROP'() {bLoopAnimation=false, sequence=/Game/Accessories/FORT_Backpacks/Backpack_M_MED_Ebony/Animation/Ebony_toStage06_PROP}
   n1 = Blend Poses by int(
    BlendPose_1=Sequence Player 'Ebony_toStage01_PROP'() {bLoopAnimation=false, sequence=/BRCosmetics/Accessories/FORT_Backpacks/Backpack_M_MED_Ebony/Animation/Ebony_toStage01_PROP}
    BlendPose_2=Sequence Player 'Ebony_toStage02_PROP'() {bLoopAnimation=false, sequence=/BRCosmetics/Accessories/FORT_Backpacks/Backpack_M_MED_Ebony/Animation/Ebony_toStage02_PROP}
    BlendPose_3=Sequence Player 'Ebony_toStage03_PROP'() {bLoopAnimation=false, sequence=/BRCosmetics/Accessories/FORT_Backpacks/Backpack_M_MED_Ebony/Animation/Ebony_toStage03_PROP}
    BlendPose_4=Sequence Player 'Ebony_toStage04_PROP'() {bLoopAnimation=false, sequence=/BRCosmetics/Accessories/FORT_Backpacks/Backpack_M_MED_Ebony/Animation/Ebony_toStage04_PROP}
    BlendPose_5=Sequence Player 'Ebony_toStage05_PROP'() {bLoopAnimation=false, sequence=/BRCosmetics/Accessories/FORT_Backpacks/Backpack_M_MED_Ebony/Animation/Ebony_toStage05_PROP}
    BlendPose_6=Sequence Player 'Ebony_toStage06_PROP_01'() {bLoopAnimation=false, sequence=/BRCosmetics/Accessories/FORT_Backpacks/Backpack_M_MED_Ebony/Animation/Ebony_toStage06_PROP_01}
    BlendTime_0=0.1
    BlendTime_1=0.1
    BlendTime_2=0.1
    BlendTime_3=0.1
    BlendTime_4=0.1
    BlendTime_5=0.1
    BlendTime_6=0.1
    ActiveChildIndex=Get KillCounter() -> (KillCounter=0)
   ) {
    blendType=Linear
    blendPose [
     {linkId=-1, sourceLinkId=-1}
     {linkId=-1, sourceLinkId=-1}
     {linkId=-1, sourceLinkId=-1}
     {linkId=-1, sourceLinkId=-1}
     {linkId=-1, sourceLinkId=-1}
     {linkId=-1, sourceLinkId=-1}
     {linkId=-1, sourceLinkId=-1}
    ]
    blendTime [0.1,0.1,0.1,0.1,0.1,0.1,0.1]
   }
   Output Animation Pose(Result=@n1)
  ]
 }
 Transition {
  path=AnimGraph.AnimGraphNode_StateMachine_0.New State Machine.AnimStateTransitionNode_0.Transition
  nodes [Result(bCanEnterTransition=Get activeKill() -> (activeKill:bool))]
 }
 Transition {path=AnimGraph.AnimGraphNode_StateMachine_0.New State Machine.AnimStateTransitionNode_1.Transition, nodes [Result()]}
 EventGraph {nodes [
   n1 = Timer Exec() {is=title K2Node_CustomEvent}
   > Update boxes visibility in Frontend()
   KillCountClampedUpdated() {is=title K2Node_CustomEvent}
   > Branch(Condition=Get IsReactivityEnabled() -> (IsReactivityEnabled:bool))
   > RetriggerableDelay(Duration=0.6)
   > Set activeKill(activeKill:bool=true) -> (Output_Get:bool)
   > n2 = Set KillCounter(KillCounter:int=Get KillCounterClamped!(self:B Athena Part Modifier Ebony Back Pack Object Reference=Get As B Athena Part Modifier Ebony Back Pack() -> (As B Athena Part Modifier Ebony Back Pack:B Athena Part Modifier Ebony Back Pack Object Reference)) -> (KillCounterClamped:int=0)) -> (Output_Get:int=0)
   > n3 = Switch on Int!(execute:Exec, Selection:int=@n2) -> (Default:Exec, 0:Exec, 1:Exec, 2:Exec)
   n4 = Get DynamicMaterialInstance() -> (DynamicMaterialInstance:Material Instance Dynamic Object Reference)
   SetScalarParameterValue(execute=@n3.1, self=@n4, ParameterName=Hide Element 02, Value=0)
   SetScalarParameterValue(execute=@n3.2, self=@n4, ParameterName=Hide Element 03, Value=0)
   > SetScalarParameterValue(self=@n4, ParameterName=Hide Element 04, Value=0)
   Event BlueprintUpdateAnimation()
   n5 = Get bodyMesh() -> (bodyMesh:Skeletal Mesh Component Object Reference)
   > Is Valid(InputObject=@n5)
   > Set PelvisLocation(PelvisLocation:Vector=GetSocketLocation!(self:Scene Component Object Reference=@n5, InSocketName:Name=pelvis) -> (ReturnValue:Vector)) -> (Output_Get:Vector)
   Event OnCharacterCustomizationLoaded()
   > n6 = Sequence()
   > n7 = Cast To FortPlayerPawn(Object=TryGetPawnOwner())
   > Set As Fort Player Pawn(As Fort Player Pawn:Fort Player Pawn Object Reference=@n7) -> (Output_Get:Fort Player Pawn Object Reference)
   n8 = GetPartSkeletalMeshComponent()
   > n9 = Get Dynamic Material Instance(SkeletalMeshComponent=@n8)
   > n10 = Set DynamicMaterialInstance(DynamicMaterialInstance:Material Instance Dynamic Object Reference=@n9) -> (Output_Get:Material Instance Dynamic Object Reference)
   > Set SkeletalMesh(SkeletalMesh:Skeletal Mesh Component Object Reference=@n8) -> (Output_Get:Skeletal Mesh Component Object Reference)
   > SetScalarParameterValue(self=@n10, ParameterName=Hide Element 02, Value=1)
   > SetScalarParameterValue(self=@n10, ParameterName=Hide Element 03, Value=1)
   > SetScalarParameterValue(self=@n10, ParameterName=Hide Element 04, Value=1)
   n11 = Get As Fort Player Pawn() -> (As Fort Player Pawn:Fort Player Pawn Object Reference)
   > Set bodyMesh(bodyMesh:Skeletal Mesh Component Object Reference=GetSkeletalMeshForPartType(self=@n11)) -> (Output_Get:Skeletal Mesh Component Object Reference)
   > n12 = GetVariantChannelActiveIndexForItemActor!(
    execute
    Item:Athena Cosmetic Item Definition Object Reference=/BRCosmetics/Athena/Items/Cosmetics/Backpacks/Backpack_Ebony
    Actor:Actor Object Reference=@n11
    Channel:Gameplay Tag Structure=(TagName="Cosmetics.Variant.Channel.Parts")
    VariantList:Array of Gameplay Tag Structures=Make Array("[0]"=(TagName="Cosmetics.Variant.Property.Stage1"))
    bErrorIfChannelMissing:bool=false
    bErrorIfTagNotInArray:bool=false
   ) -> (then, Index:int=0, bIndexIsValid:bool, bFoundChannel:bool)
   > n13 = Set IsReactivityEnabled(IsReactivityEnabled:bool=@n12.bIndexIsValid) -> (Output_Get:bool)
   > n14 = Branch(Condition=@n13)
   n15 = GetVariantChannelActiveIndexForItemActor(
    execute=@n14.else
    Item=&Backpack_Ebony
    Actor=@n11
    Channel=(TagName="Cosmetics.Variant.Channel.Parts")
    VariantList=Make Array("[0]"=(TagName="Cosmetics.Variant.Property.Stage3"))
    bErrorIfChannelMissing=false
    bErrorIfTagNotInArray=false
   ) -> (Index=0)
   > Set IsNonReactiveFullyDeployed(IsNonReactiveFullyDeployed:bool=@n15.bIndexIsValid) -> (Output_Get:bool)
   n16 = Cast To B_Athena_PartModifier_Ebony_BackPack!(
    execute=@n6.then_1
    Object=GetModifierForPartType!(self:Fort Player Pawn Object Reference=Get As Fort Player Pawn(), PartType:EFortCustomPartType Enum=Backpack) -> (ReturnValue:Custom Character Part Modifier Object Reference)
   ) -> (then, CastFailed, AsB Athena Part Modifier Ebony Back Pack:B Athena Part Modifier Ebony Back Pack Object Reference) {is=title K2Node_DynamicCast}
   > n17 = Set As B Athena Part Modifier Ebony Back Pack(As B Athena Part Modifier Ebony Back Pack:B Athena Part Modifier Ebony Back Pack Object Reference=@n16) -> (Output_Get:B Athena Part Modifier Ebony Back Pack Object Reference)
   > Set KillCounter(KillCounter=Get KillCounterClamped(self=@n17) -> (KillCounterClamped=0)) -> (Output_Get=0)
   Event BlueprintInitializeAnimation()
   > n18 = Cast To FortPlayerPawn(Object=TryGetPawnOwner())
   > Cast To B_Athena_PartModifier_Ebony_BackPack(Object=GetModifierForPartType(self=@n18, PartType=Backpack))
   Event BlueprintProcessAnimInputEvent()
   > n19 = Sequence()
   Branch(execute=@n19.then_1, Condition=Get bIsInFrontEnd())
   > Set KillCounter(KillCounter=0) -> (Output_Get=0)
   > n20 = Set Timer by Event(Delegate=@n1, Time=Get Timer() -> (Timer:double), bLooping=true)
   > n21 = Set Timer Handle(Timer Handle:Timer Handle Structure=@n20) -> (Output_Get:Timer Handle Structure)
   > Unpause Timer by Handle!(execute, Handle:Timer Handle Structure=@n21) -> (then)
   Event BlueprintPostEvaluateAnimation() {is=title K2Node_Event}
   > Set activeKill(activeKill=false)
   n22 = GetOwnerSkeletalMeshComponent()
   n23 = Set BackpackLeftTurnCorrectionAlpha(BackpackLeftTurnCorrectionAlpha:double=GetAlphaFromReferenceTransformDelta(SkeletalMeshComponent=@n22, BoneName=spine_03, BaseName=spine_01, DeltaComponentType=EulerX, InputRange=(X=-5,Y=-20), TwistAxis="1,0,0")) -> (Output_Get:double)
   > n24 = Set BackpackRightTurnCorrectionAlpha(BackpackRightTurnCorrectionAlpha:double=GetAlphaFromReferenceTransformDelta(SkeletalMeshComponent=@n22, BoneName=spine_03, BaseName=spine_01, DeltaComponentType=EulerX, InputRange=(X=5,Y=20), TwistAxis="1,0,0").RawOutput) -> (Output_Get:double)
   n25 = BreakRotator(InRot=GetDeltaTransformFromRefPose!(self=@n22, BoneName=neck_01, BaseName=spine_05) -> (ReturnValue_Location:Vector, ReturnValue_Rotation:Rotator, ReturnValue_Scale:Vector).ReturnValue_Rotation)
   > n26 = Set BackpackLookUpCorrectionAlpha(BackpackLookUpCorrectionAlpha:double=MapRangeClamped(Value="Absolute (Float)"(A=@n25), InRangeA=16, InRangeB=35)) -> (Output_Get:double)
   > Set BoxesClippingLeftTurnAlpha(BoxesClippingLeftTurnAlpha:double=float / float(A=float + float(A=@n23, B=@n26), B=2)) -> (Output_Get:double)
   > Set BoxesClippingRightTurnAlpha(BoxesClippingRightTurnAlpha:double=float / float(A=float + float(A=@n24, B=@n26), B=2)) -> (Output_Get:double)
   n27 = Get bIsTargeting()
   > Set BoxesTargetingCorrectionAlpha(BoxesTargetingCorrectionAlpha:double=float * float(A=@n26, B="To Float (Boolean)"(InBool=@n27)) {is=title K2Node_PromotableOperator}) -> (Output_Get:double)
  ]}
 Update boxes visibility in Frontend {nodes [
   Update boxes visibility in Frontend() {is=title K2Node_FunctionEntry}
   > Branch(Condition=Get IsReactivityEnabled())
   > Set activeKill(activeKill=true)
   n1 = Get KillCounter() -> (KillCounter=0)
   n2 = Get Max Kills() -> (Max Kills:int=0)
   > Set KillCounter(KillCounter="Clamp (Integer)"(Value=int + int(A=@n1), Min=0, Max=int + int(A=@n2)) -> (ReturnValue=0)) -> (Output_Get=0)
   > n3 = Switch on Int(Selection=Get KillCounter() -> (KillCounter=0))
   n4 = Get DynamicMaterialInstance()
   SetScalarParameterValue(execute=@n3.1, self=@n4, ParameterName=Hide Element 02, Value=0)
   SetScalarParameterValue(execute=@n3.2, self=@n4, ParameterName=Hide Element 03, Value=0)
   > SetScalarParameterValue(self=@n4, ParameterName=Hide Element 04, Value=0)
  ]}
]
variables [
 activeKill {type=boolean}
 As B Athena Part Modifier Ebony Back Pack {type=B_Athena_PartModifier_Ebony_BackPack_C, default=null}
 As Fort Player Pawn {type=FortPlayerPawn}
 BackpackLeftTurnCorrectionAlpha {type=number}
 BackpackLookUpCorrectionAlpha {type=number}
 BackpackRightTurnCorrectionAlpha {type=number}
 bodyMesh {type=SkeletalMeshComponent}
 Box_2_Visibility {type=number, default=1}
 Box_3_Visibility {type=number, default=1}
 BoxesClippingLeftTurnAlpha {type=number}
 BoxesClippingRightTurnAlpha {type=number}
 BoxesTargetingCorrectionAlpha {type=number}
 DynamicMaterialInstance {type=MaterialInstanceDynamic}
 IsNonReactiveFullyDeployed {type=boolean}
 IsReactivityEnabled {type=boolean}
 KillCounter {type=integer}
 Max Kills {type=integer, default=5}
 PelvisLocation {type=Vector}
 SkeletalMesh {type=SkeletalMeshComponent}
 Text_Visibility {type=number, default=1}
 Timer {type=number, default=4}
 Timer Handle {type=TimerHandle}
]
